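# Niagara trail effect for the BlastBerry umbrella glider, shedding fading sprite and mesh particles around the canopy.
# The active cubes emitter snaps random sphere UVs to a grid to spawn particles in a band.
# Particles orbit, fall and spin, with U position driving lifetime, gravity and brightness as they shrink and shift from cyan to blue.
# A disabled Heavy_Cubes emitter pushes particles along surface normals, stretched by the glider's forward dot.
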NiagaraSystem /BRCosmetics/Gadgets/Assets/VinderTech_GliderChute/Glider_Umbrella_BlastBerry/FX/NS_Glider_Umbrella_BlastBerry_Trail
dependencies {
 usedDataInterfaces [&NiagaraDataInterfaceCurve, &NiagaraDataInterfaceColorCurve]
 usedDynamicInputs [&MaskFloatBySpawnGroup, &Abs_Float]
 usedModules [
  &SystemState
  &InterpolateOverTime
  &self:SystemSpawnScript.SystemScriptSource.SystemScriptGraph.NiagaraNodeAssignment_2.SetVariables_#1
 ]
}
emitters {
 Em_Umbrella_Glider_Cubes {
  emitterSpawnScript [/SetVariables(Emitter.GridSizeXY:Vector2f=<20 8>) {isSetParameters=true}]
  emitterUpdateScript [
   &EmitterState()
   &SpawnRate(SpawnRate=128, Use Spawn Probability=false)
  ]
  particleSpawnScript [
   /SetVariables(Particles.RandomU:NiagaraFloat=&FloatFromCurve(
     FloatCurve=DataInterface() {propertyValues {Curve {keys [
         {interpMode=RCIM_Cubic, tangentMode=RCTM_User, time=0, value=0}
         {arriveTangent=2.2285, interpMode=RCIM_Cubic, leaveTangent=2.2285, tangentMode=RCTM_User, time=1, value=1}
        ]}}}
     CurveIndex=&RandomRangeFloat(Minimum=0, Maximum=1)
    )) {isSetParameters=true}
   /SetVariables(Particles.RandomV:NiagaraFloat=&RandomRangeFloat(Minimum=0, Maximum=1)) {isSetParameters=true}
   /SetVariables(Particles.UPositionRange:NiagaraFloat=&Add_Float(
     A=-0.155
     B=&Multiply_Float(
      A=&Divide_Float(
       A=/Niagara/DynamicInputs/Math/Floor_Float(Float:NiagaraFloat=&Multiply_Float(A=$Particles.RandomU, B=&MakeFloatFromVector2d(Vector 2D=$Emitter.GridSizeXY)))
       B=&MakeFloatFromVector2d(Vector 2D=$Emitter.GridSizeXY)
      )
      B=-0.65
     )
    )) {isSetParameters=true}
   /SetVariables(Particles.VPositionRange:NiagaraFloat=&Add_Float(
     A=0.91
     B=&Multiply_Float(
      A=&Divide_Float(
       A=&Ceil_Float(Float=&Multiply_Float(
         A=$Particles.RandomV
         B=&MakeFloatFromVector2d(Vector 2D=$Emitter.GridSizeXY, Channel=::NewEnumerator1)
        ))
       B=&MakeFloatFromVector2d(Vector 2D=$Emitter.GridSizeXY, Channel=::NewEnumerator1)
      )
      B=0.087
     )
    )) {isSetParameters=true}
   &InitializeParticle(
    Lifetime Max=1.75
    Lifetime Min=1.4
    Lifetime Mode=::NewEnumerator1
    Mass Mode=::NewEnumerator3
    Mesh Scale Mode=::NewEnumerator3
    Mesh Uniform Scale=2
    Sprite Rotation Mode=::NewEnumerator2
    Sprite Size Mode=::NewEnumerator1
    Uniform Sprite Size Max=12
    Uniform Sprite Size Min=6
    UsePositionOffset=false
    Write Material Random=true
   )
   &ShapeLocation(
    MaskSpawnGroup=false
    Non Uniform Scale=<1 1 0.95>
    Offset=<0 0 -98>
    Offset Mode=::NewEnumerator0
    Rotation Mode=::NewEnumerator2
    Sphere Distribution=::NewEnumerator1
    Sphere Radius=110
    U Position=$Particles.UPositionRange
    UseOwnerScale=true
    V Position=$Particles.VPositionRange
    Yaw / Pitch / Roll=&MakeVector(X=&Multiply_Float(A=$Engine.System.Age, B=0), Y=0, Z=0)
   ) {module=ShapeLocation001}
   &InitialMeshOrientation(
    Mesh Axis To Orient=/Niagara/DynamicInputs/Vectors/CrossProduct(
     Vector A:Vector3f=&MakeVector(
      X=&MakeFloatFromVector(Vector=$Particles.ShapeLocation001.ShapeNormal, Channel=::NewEnumerator0)
      Y=&MakeFloatFromVector(Vector=$Particles.ShapeLocation001.ShapeNormal, Channel=::NewEnumerator1)
      Z=&MakeFloatFromVector(Vector=$Particles.ShapeLocation001.ShapeNormal, Channel=::NewEnumerator2)
     )
     Vector B:Vector3f=&MakeVector(
      X=&MakeFloatFromVector(Vector=$Particles.ShapeLocation001.ShapeTangent, Channel=::NewEnumerator0)
      Y=&MakeFloatFromVector(Vector=$Particles.ShapeLocation001.ShapeTangent, Channel=::NewEnumerator1)
      Z=&MakeFloatFromVector(Vector=$Particles.ShapeLocation001.ShapeTangent, Channel=::NewEnumerator2)
     )
     Normalize Incoming Vectors:NiagaraBool=true
    )
    Mesh Orientation Mode=::NewEnumerator2
    Orientation Coordinate Space=::Local
    Orientation Vector=<0 0 1>
    Rotation=<0 0 0.125>
    Use Rotation Vector=true
   )
  ]
  particleUpdateScript [
   &ParticleState(Lifetime:NiagaraFloat=&Multiply_Float(
     A=$Particles.Lifetime
     B=&Multiply_Float(A=&OneMinusFloat(Float=$Particles.UPositionRange), B=0.75)
    ))
   &self:ManualOrbit(RotationSpeed:NiagaraFloat=&FloatFromCurve(FloatCurve=DataInterface() {propertyValues {Curve {keys [
         {arriveTangent=-6.7477, interpMode=RCIM_Cubic, leaveTangent=-6.7477, tangentMode=RCTM_User, time=0, value=0}
         {tangentMode=RCTM_User, time=1, value=-3}
        ]}}})) {module=ManualOrbit}
   &GravityForce(Gravity=&MakeVector(X=0, Y=0, Z=&Multiply_Float(A=$Particles.UPositionRange, B=-50)))
   &AddRotationalVelocity(Rotation Rate=&VectorFromFloat(Value=&FloatFromCurve(
      FloatCurve=DataInterface() {propertyValues {Curve {keys [
          <at 0 0 RCIM_Cubic>
          {arriveTangent=0.4906, interpMode=RCIM_Cubic, leaveTangent=0.4906, tangentMode=RCTM_User, time=0.5, value=0.05}
          {arriveTangent=2.5455, interpMode=RCIM_Cubic, leaveTangent=2.5455, tangentMode=RCTM_User, time=1, value=2}
         ]}}}
      Scale Curve=0.1
     )))
   &VortexForce(UseOriginOffset=true, Vortex Force Amount=-50) {enabled=false}
   &Drag(Drag=0.25, Use Linear Drag=true, Use Rotational Drag=true)
   &SolveForcesAndVelocity(Clamp Velocity=false, Limit Acceleration=false)
   &ScaleColor(
    Color Value To Scale=&MultiplyLinearColor(
     Linear Color=$Particles.Initial.Color
     Scale Factor=&MakeFloat4FromVectorAndFloat(
      "Vector (RGB)"=&VectorFromFloat(Value=&Multiply_Float(
        A=&OneMinusFloat(Float=/Game/Effects/Niagara/DynamicInputs/Math/FN_Fit01(Clamp:NiagaraBool=false, Value:NiagaraFloat=$Particles.UPositionRange, DestMin:NiagaraFloat=0, DestMax:NiagaraFloat=-1))
        B=1
       ))
      "Float (Alpha)"=1
     )
    )
    Scale Alpha=&FloatFromCurve(FloatCurve=DataInterface() {propertyValues {Curve {keys [<at 0 1>,<at 1 0>]}}})
    Scale RGB=&MakeVectorFromLinearColorRGB(Color=&ColorFromCurve(ColorCurve=DataInterface() {propertyValues {
        AlphaCurve {keys [<at 0 1 RCIM_Cubic>,<at 1 0 RCIM_Cubic>]}
        BlueCurve {keys [<at 0 1 RCIM_Cubic>,<at 0.508 1>,<at 1 1 RCIM_Cubic>]}
        GreenCurve {keys [<at 0 0.974 RCIM_Cubic>,<at 0.508 0.4425>,<at 1 0.1946 RCIM_Cubic>]}
        RedCurve {keys [<at 0 0 RCIM_Cubic>,<at 0.508 0>,<at 1 0.1946 RCIM_Cubic>]}
       }}))
    ScaleA=true
    ScaleRGB=true
   ) {module=ScaleColor001}
   &ScaleColor(
    Color Value To Scale=&MultiplyLinearColor(
     Linear Color=$Particles.Initial.Color
     Scale Factor=&MakeFloat4FromVectorAndFloat(
      "Vector (RGB)"=&VectorFromFloat(Value=&Multiply_Float(
        A=&OneMinusFloat(Float=&FN_Fit01(Clamp:NiagaraBool=false, Value:NiagaraFloat=$Particles.UPositionRange, DestMin:NiagaraFloat=0, DestMax:NiagaraFloat=-1))
        B=1
       ))
      "Float (Alpha)"=1
     )
    )
    Linear Color Curve=DataInterface() {propertyValues {
      AlphaCurve {keys [<at 0 1>,<at 1 0>]}
      BlueCurve {keys [<at 0 1>,<at 1 1>]}
      GreenCurve {keys [<at 0 0.9823>,<at 1 0.194>]}
      RedCurve {keys [<at 0 0.3185>,<at 1 0.194>]}
     }}
    Scale Mode=::NewEnumerator2
   ) {enabled=false}
   &ScaleMeshSize(Scale Factor=&VectorFromFloat(Value=&FloatFromCurve(FloatCurve=DataInterface() {propertyValues {Curve {keys [<at 0 1 RCIM_Cubic>, {arriveTangent=-3.5633, interpMode=RCIM_Cubic, leaveTangent=-3.5633, tangentMode=RCTM_User, time=1, value=0}]}}})))
   &DynamicMaterialParameters(
    Index 0 Param 1=&MakeCustomFloatFromBool(
     Bool=true
     True Float=&FloatFromCurve(FloatCurve=DataInterface() {propertyValues {Curve {keys [{arriveTangent=-2.1994, interpMode=RCIM_Cubic, leaveTangent=-2.1994, tangentMode=RCTM_User, time=0, value=1}, <at 1 0>]}}})
     False Float=&FloatFromCurve(
      FloatCurve=DataInterface() {propertyValues {Curve {keys [{arriveTangent=-2.1994, interpMode=RCIM_Cubic, leaveTangent=-2.1994, tangentMode=RCTM_User, time=0, value=1}, <at 1 0>]}}}
      Scale Curve=0.5
     )
    )
    Module.Index0Param0Write=true
    Module.Index0Param1Write=true
    Module.Index0Param2Write=true
    Module.Index0Param3Write=true
   )
  ]
  rendererClasses [&NiagaraSpriteRendererProperties, &NiagaraMeshRendererProperties]
  renderers [
   {rendererClass=&NiagaraSpriteRendererProperties}
   {rendererClass=&NiagaraMeshRendererProperties, rendererIndex=1}
  ]
 }
 Heavy_Cubes {
  bEnabled=false
  emitterSpawnScript [/SetVariables(Emitter.GridSizeXY:Vector2f=<20 8>) {isSetParameters=true}]
  emitterUpdateScript [
   &EmitterState()
   &SpawnRate(SpawnRate=20, Use Spawn Probability=false)
   &SpawnRate(SpawnGroup=1, SpawnRate=15, Use Spawn Probability=false) {module=SpawnRate001}
  ]
  particleSpawnScript [
   /SetVariables(Particles.VisibilityTag=&SpawnGroup()) {isSetParameters=true}
   /SetVariables(Particles.RandomU:NiagaraFloat=&FloatFromCurve(
     FloatCurve=DataInterface() {propertyValues {Curve {keys [
         {interpMode=RCIM_Cubic, tangentMode=RCTM_User, time=0, value=0}
         {arriveTangent=2.2285, interpMode=RCIM_Cubic, leaveTangent=2.2285, tangentMode=RCTM_User, time=1, value=1}
        ]}}}
     CurveIndex=&RandomRangeFloat(Minimum=0, Maximum=1)
    )) {isSetParameters=true}
   /SetVariables(Particles.RandomV:NiagaraFloat=&RandomRangeFloat(Minimum=0, Maximum=1)) {isSetParameters=true}
   /SetVariables(Particles.UPositionRange:NiagaraFloat=&Add_Float(
     A=-0.155
     B=&Multiply_Float(
      A=&Divide_Float(
       A=&Floor_Float(Float:NiagaraFloat=&Multiply_Float(A=$Particles.RandomU, B=&MakeFloatFromVector2d(Vector 2D=$Emitter.GridSizeXY)))
       B=&MakeFloatFromVector2d(Vector 2D=$Emitter.GridSizeXY)
      )
      B=-0.5
     )
    )) {isSetParameters=true}
   /SetVariables(Particles.VPositionRange:NiagaraFloat=&Add_Float(
     A=0.91
     B=&Multiply_Float(
      A=&Divide_Float(
       A=&Ceil_Float(Float=&Multiply_Float(
         A=$Particles.RandomV
         B=&MakeFloatFromVector2d(Vector 2D=$Emitter.GridSizeXY, Channel=::NewEnumerator1)
        ))
       B=&MakeFloatFromVector2d(Vector 2D=$Emitter.GridSizeXY, Channel=::NewEnumerator1)
      )
      B=0.087
     )
    )) {isSetParameters=true}
   &InitializeParticle(
    Color Maximum=<rgba 0.209 0.6456 0.66 1>
    Color Minimum=<rgba 0.1045 0.3228 0.33 1>
    Color Mode=::NewEnumerator2
    Lifetime Max=0.75
    Lifetime Min=0.5
    Lifetime Mode=::NewEnumerator1
    Mesh Scale Mode=::NewEnumerator3
    Mesh Uniform Scale=&Lerp_Float(A=4, B=1, Alpha=&Lerp_Float(A=$Particles.RandomV, B=1, Alpha=$Particles.RandomU))
    Position Offset=<0 0 -90>
    Sprite Size Mode=::NewEnumerator3
    Uniform Sprite Size=&Lerp_Float(A=5, B=10, Alpha=&RandomRangeFloat(Minimum=0, Maximum=1))
    UsePositionOffset=true
    Write Material Random=true
   )
   &ShapeLocation(
    MaskSpawnGroup=false
    Non Uniform Scale=<1 1 0.95>
    Sphere Distribution=::NewEnumerator1
    U Position=$Particles.UPositionRange
    UseOwnerScale=true
    V Position=$Particles.VPositionRange
   )
   &DynamicMaterialParameters(
    Index 0 Param 1=&UniformRangedFloat(Minimum=0, Maximum=1)
    Index 0 Param 2=&UniformRangedFloat(Minimum=1, Maximum=2)
    Index 0 Param 3=1
    Module.Param0WriteEnabled=true
    Module.Param1WriteEnabled=true
    Module.Param2WriteEnabled=false
    Module.Param3WriteEnabled=false
   )
  ]
  particleUpdateScript [
   &ParticleState()
   /SetVariables(Particles.Direction:Vector3f=&NormalizeVector(
     Vector To Normalize=&Multiply_Vector(
      A=$Particles.Initial.ShapeLocation.ShapeNormal
      B=&Lerp_Vector(
       A=<1 1 1>
       B=<10 10 0.25>
       Alpha=&Multiply_Float(A=$System.ForwardDot, B=&Power(a=$Particles.NormalizedAge, b=2))
      )
     )
     Safe Normalize=false
     Safe Fallback Threshold=0
     Safe Fallback Vector=<0 0 0.001>
    )) {isSetParameters=true}
   /SetVariables(Particles.Velocity=$Particles.Direction) {isSetParameters=true}
   /SetVariables(Particles.Position=&AddVectorToPosition(
     Position=$Particles.Initial.Position
     Vector=&Multiply_VectorByFloat(
      Vector=$Particles.Direction
      Float=&Multiply_Float(
       A=&Power(a=$Particles.NormalizedAge, b=1)
       B=&Multiply_Float(A=$Particles.Initial.Lifetime, B=&Lerp_Float(A=30, B=70, Alpha=$System.ForwardDot))
      )
     )
    )) {isSetParameters=true}
   &UpdateMeshOrientation(
    Facing Direction=$Particles.Initial.ShapeLocation.ShapeTangent
    Facing Interpolation=false
    Orientation Method=::NewEnumerator1
    Override Quat=false
    TransformFacing=false
    TransformUp=false
    Up Direction:Vector3f=$Particles.Direction
   )
   &ScaleColor(
    Scale Alpha=&FloatFromCurve(FloatCurve=DataInterface() {propertyValues {Curve {keys [<at 0.66 1 RCIM_Cubic>, {arriveTangent=-4.6302, interpMode=RCIM_Cubic, leaveTangent=-4.6302, tangentMode=RCTM_User, time=1, value=0}]}}})
    Scale RGB=&VectorFromFloat(Value=&FloatFromCurve(FloatCurve=DataInterface() {propertyValues {Curve {keys [
          {arriveTangent=-138.954, interpMode=RCIM_Cubic, leaveTangent=-138.954, tangentMode=RCTM_User, time=0, value=10}
          <at 0.1 1>
          {interpMode=RCIM_Cubic, leaveTangent=-29.5004, time=1, value=0}
         ]}}}))
    ScaleA=true
    ScaleRGB=true
   )
   /SetVariables(Particles.Color_Sprite:LinearColor=&MakeLinearColorFromVectorAndFloat(
     "Vector (RGB)"=&Multiply_VectorByFloat(
      Vector=&MakeVectorFromLinearColorRGB(Color=$Particles.Initial.Color)
      Float=&FloatFromCurve(FloatCurve=DataInterface() {propertyValues {Curve {keys [
           {arriveTangent=-138.954, interpMode=RCIM_Cubic, leaveTangent=-138.954, tangentMode=RCTM_User, time=0, value=10}
           <at 0.1 1.5>
           {interpMode=RCIM_Cubic, leaveTangent=-29.5004, time=1, value=0.5}
          ]}}})
     )
     "Float (Alpha)"=&FloatFromCurve(FloatCurve=DataInterface() {propertyValues {Curve {keys [
          {arriveTangent=9.0725, interpMode=RCIM_Cubic, leaveTangent=9.0725, tangentMode=RCTM_User, time=0, value=0}
          {interpMode=RCIM_Cubic, tangentMode=RCTM_User, time=0.1408, value=0.9847}
          {arriveTangent=-0.0405, interpMode=RCIM_Cubic, leaveTangent=-0.0405, tangentMode=RCTM_User, time=1, value=0}
         ]}}})
    )) {isSetParameters=true}
   &ScaleMeshSize(Scale Factor=&VectorFromFloat(Value=&FloatFromCurve(FloatCurve=DataInterface() {propertyValues {Curve {keys [
          {arriveTangent=13.966, interpMode=RCIM_Cubic, leaveTangent=13.966, tangentMode=RCTM_User, time=0, value=0.1}
          {interpMode=RCIM_Cubic, tangentMode=RCTM_User, time=0.15, value=1}
          {tangentMode=RCTM_User, time=1, value=0.1}
         ]}}})))
   &ScaleSpriteSize(
    Scale Factor=&Lerp_Vector2(
     A=<1 1>
     B=<0.25 1.5>
     Alpha=&FloatFromCurve(FloatCurve=DataInterface() {propertyValues {Curve {keys [
          {arriveTangent=4.1785, interpMode=RCIM_Cubic, leaveTangent=4.1785, tangentMode=RCTM_User, time=0, value=0}
          {arriveTangent=1.1179, interpMode=RCIM_Cubic, leaveTangent=1.1179, tangentMode=RCTM_User, time=0.2042, value=0.6228}
          {arriveTangent=0.0224, interpMode=RCIM_Cubic, leaveTangent=0.0224, tangentMode=RCTM_User, time=1, value=1}
         ]}}})
    )
    Scale Sprite Size Mode=::NewEnumerator2
   )
  ]
  rendererClasses [&NiagaraMeshRendererProperties, &NiagaraSpriteRendererProperties]
  renderers [
   {rendererClass=&NiagaraMeshRendererProperties, rendererIndex=0}
   {rendererClass=&NiagaraSpriteRendererProperties, rendererIndex=1}
  ]
 }
}
summary {
 systemName=NS_Glider_Umbrella_BlastBerry_Trail
 emitters [
  {
   bEnabled=false
   emitterName=Heavy_Cubes
   rendererClasses [&NiagaraMeshRendererProperties, &NiagaraSpriteRendererProperties]
  }
  {
   emitterName=Em_Umbrella_Glider_Cubes
   rendererClasses [&NiagaraSpriteRendererProperties, &NiagaraMeshRendererProperties]
  }
 ]
 userVariables [
  User.bIsFrontEndPreview {type=NiagaraBool}
  User.bIsFullyDeployed {type=NiagaraBool}
  User.ForwardDot {type=NiagaraFloat, defaultValue=0, description=""}
  User.RightDot {type=NiagaraFloat, defaultValue=0, description=""}
 ]
}
